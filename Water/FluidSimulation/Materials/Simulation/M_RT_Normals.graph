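# Unlit alpha-composite fluid simulation material that turns the height field in the RT texture into surface normals.
# Samples red-channel heights one texel either side in X and Y, then a Custom node crosses central-difference tangents scaled by FluidSize over Resolution.
# Packs normal XY with the +X sample's height (labelled Midpoint) and saturates that sample's green channel.
# The Custom node's HeightScale, SplineMargin and UV inputs are unused.
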Material /FortWater/FluidSimulation/Materials/Simulation/M_RT_Normals
graph {
 properties {
  bAllowNegativeEmissiveColor=true
  bScreenSpaceReflections=true
  bUsedWithNanite=true
  blendMode=BLEND_AlphaComposite
  pixelDepthOffsetMode=PDOM_Legacy
  shadingModel=MSM_Unlit
  translucencyLightingMode=TLM_Surface
  usedShadingModels=Unlit
 }
 nodes [
  MaterialExpressionScalarParameter_3 = ScalarParameter() {defaultValue=1, parameterName=hs}
  MaterialExpressionScalarParameter_4 = ScalarParameter() {defaultValue=0.2, parameterName=sm}
  n1 = TextureCoordinate()
  n2 = MaterialFunctionCall(Float2=Divide(B=ScalarParameter() {defaultValue=1024, parameterName=Resolution}) {desc=CellSize}) {materialFunction=&BreakOutFloat2Components}
  n3 = Constant()
  n4 = AppendVector(A=@n2.R, B=@n3) {desc=X offset}
  n5 = TextureSampleParameter2D(UVs=Add(A=@n1, B=@n4)) {desc=Midpoint, parameterName=RT, texture=&Black}
  n6 = AppendVector(A=@n3, B=@n2.G) {desc=Y offset}
  MaterialExpressionMaterialFunctionCall_1 = MaterialFunctionCall(
   X=@n5.R
   Y=TextureSampleParameter2D(UVs=Subtract(A=@n1, B=@n4)) {desc=N2, parameterName=RT, texture=&Black}.R
   Z=TextureSampleParameter2D(UVs=Add(A=@n1, B=@n6)) {desc=N3, parameterName=RT, texture=&Black}.R
   A=TextureSampleParameter2D(UVs=Subtract(A=@n1, B=@n6)) {desc=N4, parameterName=RT, texture=&Black}.R
  ) {desc=Surrounding Heights, materialFunction=&MakeFloat4}
  MaterialExpressionDivide_1 = Divide(
   A=AppendVector(A=ScalarParameter() {defaultValue=512, parameterName=FluidSize}, B=ScalarParameter() {defaultValue=512, parameterName=FluidSize})
   B=ScalarParameter() {defaultValue=1024, parameterName=Resolution}
  )
  MaterialExpressionTextureCoordinate_5 = TextureCoordinate()
  n7 = Custom!(
   HeightScale=@MaterialExpressionScalarParameter_3
   SplineMargin=@MaterialExpressionScalarParameter_4
   SurroundingHeights=@MaterialExpressionMaterialFunctionCall_1.Result
   CellSize=@MaterialExpressionDivide_1
   UV=@MaterialExpressionTextureCoordinate_5
  ) {
   code="//input float HeightScale;\r\n//input float SplineMargin;\r\n\r\n    float2 HeightDeltas = (SurroundingHeights.xz - SurroundingHeights.yw);\r\n    float3 Tangent = float3(CellSize.x * 2.0f, 0, HeightDeltas.x);\r\n    float3 Binormal = float3(0, CellSize.y * 2.0f, HeightDeltas.y);\r\n    //float3 Cross = cross(Tangent, Binormal);\r\n    // Cross product taking advantage of the 0's in Tangent and Binormal\r\n    float3 Cross = float3(-1, -1, 1) * Tangent.zxx * Binormal.yzy;\r\n    float3 Normal = normalize(Cross);\r\n\r\nreturn Normal;"
   description=Height to Normals
   inputs [
    {inputName=HeightScale, input {expression=&self:MaterialExpressionScalarParameter_3}}
    {inputName=SplineMargin, input {expression=&self:MaterialExpressionScalarParameter_4}}
    {
     inputName=SurroundingHeights
     input {expression=&self:MaterialExpressionMaterialFunctionCall_1}
    }
    {inputName=CellSize, input {expression=&self:MaterialExpressionDivide_1}}
    {inputName=UV, input {expression=&self:MaterialExpressionTextureCoordinate_5}}
   ]
  }
  AppendVector(A=ComponentMask(None=@n7) {g=true, r=true}, B=@n5.R)
  Saturate(None=@n5.G)
 ]
}
